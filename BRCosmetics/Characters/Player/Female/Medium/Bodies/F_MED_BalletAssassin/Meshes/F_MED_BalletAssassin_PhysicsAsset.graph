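# Physics asset for the BalletAssassin female medium player cosmetic body, covering its skirt bones and lower body.
# Five two-link dynamic skirt chains hang from tiny base capsules tied to the pelvis, each link constrained with tight swing and near-zero twist limits.
# Larger pelvis, thigh and foot capsules supply body volume, joined by free angular constraints, with feet linked directly to thighs and no calf bodies.
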PhysicsAsset /BRCosmetics/Characters/Player/Female/Medium/Bodies/F_MED_BalletAssassin/Meshes/F_MED_BalletAssassin_PhysicsAsset
bodies {
 base_skirt_0 [{shapeType=Capsule, length=0.505, radius=0.505, shapeName=base_skirt_0_capsule, rotation=<rot 0 -90 0>}]
 base_skirt_a_00_l [{shapeType=Capsule, length=0.505, radius=0.505, shapeName=base_skirt_a_00_l_capsule, rotation=<rot 0 -90 0>}]
 base_skirt_a_00_r [{shapeType=Capsule, length=0.505, radius=0.505, shapeName=base_skirt_a_00_r_capsule, rotation=<rot 0 -90 0>}]
 base_skirt_c_00_l [{shapeType=Capsule, length=0.505, radius=0.505, shapeName=base_skirt_c_00_l_capsule, rotation=<rot 0 -90 0>}]
 base_skirt_c_00_r [{shapeType=Capsule, length=0.505, radius=0.505, shapeName=base_skirt_c_00_r_capsule, rotation=<rot 0 -90 0>}]
 dyn_skirt_00 [{shapeType=Capsule, length=10.7949, radius=1, shapeName=dyn_skirt_00_capsule, center=<5.8606 0 0>, rotation=<rot 90 0 0>}]
 dyn_skirt_01 [{shapeType=Capsule, length=13.6208, radius=1, shapeName=dyn_skirt_01_capsule, center=<7.0236 0 0>, rotation=<rot 90 0 0>}]
 dyn_skirt_a_00_l1 [{shapeType=Capsule, length=8.358, radius=1, shapeName=dyn_skirt_a_00_l1_capsule, center=<4.6128 0 0>, rotation=<rot 90 0 0>}]
 dyn_skirt_a_00_r1 [{shapeType=Capsule, length=7.7945, radius=1, shapeName=dyn_skirt_a_00_r1_capsule, center=<4.5111 0 0>, rotation=<rot 90 0 0>}]
 dyn_skirt_a_01_l [{shapeType=Capsule, length=8.3291, radius=1, shapeName=dyn_skirt_a_01_l_capsule, center=<3.9985 0.6584 0>, rotation=<rot 90 46.9346 38.197>}]
 dyn_skirt_a_01_r [{shapeType=Capsule, length=10.0506, radius=1, shapeName=dyn_skirt_a_01_r_capsule, center=<4.9963 0 0>, rotation=<rot 90 0 0>}]
 dyn_skirt_c_00_l [{shapeType=Capsule, length=2, radius=1, shapeName=dyn_skirt_c_00_l_capsule, center=<2 0 0>, rotation=<rot 90 0 0>}]
 dyn_skirt_c_00_r [{shapeType=Capsule, length=2, radius=1, shapeName=dyn_skirt_c_00_r_capsule, center=<2 0 0>, rotation=<rot 90 0 0>}]
 dyn_skirt_c_01_l [{shapeType=Capsule, length=6.1162, radius=1, shapeName=dyn_skirt_c_01_l_capsule, center=<3.2779 0.0677 -2.2041>, rotation=<rot 90 0 0>}]
 dyn_skirt_c_01_r [{shapeType=Capsule, length=5.2069, radius=1, shapeName=dyn_skirt_c_01_r_capsule, center=<2.6968 0.0071 3.4006>, rotation=<rot 90 0 0>}]
 foot_l [{shapeType=Capsule, length=13.7001, radius=6.8798, shapeName=foot_l_capsule, center=<-2.2275 7.6636 0>, rotation=<rot 90 180 -90>}]
 foot_r [{shapeType=Capsule, length=13.7001, radius=6.8798, shapeName=foot_r_capsule, center=<2.2409 -7.6636 0>, rotation=<rot 90 153.4349 -116.5651>}]
 pelvis [{shapeType=Capsule, length=14.7514, radius=7.9514, shapeName=pelvis_capsule, center=<-4.8534 -2.0565 0.0667>, rotation=<rot 0 -0.0072 0>}]
 thigh_l [{shapeType=Capsule, length=22.5192, radius=8.2863, shapeName=thigh_l_capsule, center=<-17.806 -0.3515 0.7532>, rotation=<rot 90 0 0>}]
 thigh_r [{shapeType=Capsule, length=22.3989, radius=8.4066, shapeName=thigh_r_capsule, center=<17.806 -0.3515 -0.2652>, rotation=<rot 90 0 0>}]
}
constraints [
 {bone1Name=dyn_skirt_00, bone2Name=base_skirt_0, swing1LimitDegrees=25, swing2LimitDegrees=10, twistLimitDegrees=1}
 {bone1Name=dyn_skirt_01, bone2Name=dyn_skirt_00, swing1LimitDegrees=25, swing2LimitDegrees=10, twistLimitDegrees=1}
 {bone1Name=dyn_skirt_a_00_l1, bone2Name=base_skirt_a_00_l, swing1LimitDegrees=25, swing2LimitDegrees=10, twistLimitDegrees=1}
 {bone1Name=dyn_skirt_a_01_l, bone2Name=dyn_skirt_a_00_l1, swing1LimitDegrees=25, swing2LimitDegrees=10, twistLimitDegrees=1}
 {bone1Name=dyn_skirt_a_00_r1, bone2Name=base_skirt_a_00_r, swing1LimitDegrees=25, swing2LimitDegrees=10, twistLimitDegrees=1}
 {bone1Name=dyn_skirt_a_01_r, bone2Name=dyn_skirt_a_00_r1, swing1LimitDegrees=25, swing2LimitDegrees=10, twistLimitDegrees=1}
 {bone1Name=dyn_skirt_c_00_l, bone2Name=base_skirt_c_00_l, swing1LimitDegrees=25, swing2LimitDegrees=10, twistLimitDegrees=1}
 {bone1Name=dyn_skirt_c_01_l, bone2Name=dyn_skirt_c_00_l, swing1LimitDegrees=25, swing2LimitDegrees=10, twistLimitDegrees=1}
 {bone1Name=dyn_skirt_c_00_r, bone2Name=base_skirt_c_00_r, swing1LimitDegrees=25, swing2LimitDegrees=10, twistLimitDegrees=1}
 {bone1Name=dyn_skirt_c_01_r, bone2Name=dyn_skirt_c_00_r, swing1LimitDegrees=25, swing2LimitDegrees=10, twistLimitDegrees=1}
 {
  bone1Name=thigh_l
  bone2Name=pelvis
  swing1LimitDegrees=45
  swing1Motion=Free
  swing2LimitDegrees=45
  swing2Motion=Free
  twistLimitDegrees=45
  twistMotion=Free
 }
 {
  bone1Name=thigh_r
  bone2Name=pelvis
  swing1LimitDegrees=45
  swing1Motion=Free
  swing2LimitDegrees=45
  swing2Motion=Free
  twistLimitDegrees=45
  twistMotion=Free
 }
 {
  bone1Name=base_skirt_0
  bone2Name=pelvis
  swing1LimitDegrees=45
  swing1Motion=Free
  swing2LimitDegrees=45
  swing2Motion=Free
  twistLimitDegrees=45
  twistMotion=Free
 }
 {
  bone1Name=base_skirt_a_00_l
  bone2Name=pelvis
  swing1LimitDegrees=45
  swing1Motion=Free
  swing2LimitDegrees=45
  swing2Motion=Free
  twistLimitDegrees=45
  twistMotion=Free
 }
 {
  bone1Name=base_skirt_a_00_r
  bone2Name=pelvis
  swing1LimitDegrees=45
  swing1Motion=Free
  swing2LimitDegrees=45
  swing2Motion=Free
  twistLimitDegrees=45
  twistMotion=Free
 }
 {
  bone1Name=base_skirt_c_00_r
  bone2Name=pelvis
  swing1LimitDegrees=45
  swing1Motion=Free
  swing2LimitDegrees=45
  swing2Motion=Free
  twistLimitDegrees=45
  twistMotion=Free
 }
 {
  bone1Name=base_skirt_c_00_l
  bone2Name=pelvis
  swing1LimitDegrees=45
  swing1Motion=Free
  swing2LimitDegrees=45
  swing2Motion=Free
  twistLimitDegrees=45
  twistMotion=Free
 }
 {
  bone1Name=foot_l
  bone2Name=thigh_l
  swing1LimitDegrees=45
  swing1Motion=Free
  swing2LimitDegrees=45
  swing2Motion=Free
  twistLimitDegrees=45
  twistMotion=Free
 }
 {
  bone1Name=foot_r
  bone2Name=thigh_r
  swing1LimitDegrees=45
  swing1Motion=Free
  swing2LimitDegrees=45
  swing2Motion=Free
  twistLimitDegrees=45
  twistMotion=Free
 }
]
