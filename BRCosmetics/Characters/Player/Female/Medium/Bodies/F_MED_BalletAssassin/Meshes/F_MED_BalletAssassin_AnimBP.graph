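# Anim Blueprint for the Ballet Assassin body part, animating its skirt with bone physics and shader-driven cloth.
# Copied parent-mesh pose passes through a Control Rig and attach_cape offset, then RigidBody simulation, skipped when a cape or long back bling is worn.
# Front end uses separate physics tuning; skydiving lowers in-game simulation and offset alpha.
# Outside front end, updates feed leg motion to WPO cloth materials, raising wind and turbulence while gliding.
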AnimBlueprint /BRCosmetics/Characters/Player/Female/Medium/Bodies/F_MED_BalletAssassin/Meshes/F_MED_BalletAssassin_AnimBP
target_skeleton=&/F_MED_BalletAssassin_Skeleton
graphs [
 AnimGraph {nodes [
   Output Pose(Result=Use cached pose 'Skirt Dynamics'()) {layerGroup=None}
   n1 = Control Rig(Source=Copy Pose From Mesh() {bUseAttachedParent=true}) {
    alphaBoolBlend {blendOption=Linear}
    controlRigAssetReference {blueprintRigClass=&/F_MED_BalletAssasin_Master_ControlRig.F_MED_BalletAssasin_Master_ControlRig_C}
   }
   Save cached pose 'Start Pose'(Pose=@n1) {cacheName=Start Pose}
   n2 = RigidBody!(
    ComponentPose=Local To Component(LocalPose=Use cached pose 'Skydive Cape '())
    Alpha="Get is Skydiving "() -> ("is Skydiving ":double)
    OverrideWorldGravity:Vector=Get GravityOverride()
    ExternalForce
   ) -> (Pose) {
    bOverrideWorldGravity=true
    overridePhysicsAsset=&/F_MED_BalletAssassin_PhysicsAsset
    simulationSpace=BaseBoneSpace
    alphaBoolBlend {blendOption=Linear}
    alphaScaleBiasClamp {bInterpResult=true}
    baseBoneRef {boneName=pelvis}
    componentAppliedLinearAccClamp=<3000 3000 3000>
    componentLinearAccScale=<0.3 0.2 0.2>
    componentLinearVelScale=<0.3 0.2 0.2>
    simSpaceSettings {
     maxAngularAcceleration=1000
     maxAngularVelocity=3
     maxLinearAcceleration=700
     maxLinearVelocity=500
     velocityScaleZ=0
     worldAlpha=0.2
     externalLinearDragV=<0.2 0.2>
    }
   }
   n3 = Component To Local(ComponentPose=@n2)
   n4 = Get bIsInFrontEnd()
   n5 = RigidBody(ComponentPose=Local To Component(LocalPose=Use cached pose 'Skydive Cape '()), Alpha=0.65, OverrideWorldGravity=Get GravityOverride()) {
    bOverrideWorldGravity=true
    overridePhysicsAsset=&F_MED_BalletAssassin_PhysicsAsset
    simulationSpace=BaseBoneSpace
    alphaBoolBlend {blendOption=Linear}
    baseBoneRef {boneName=pelvis}
    componentAppliedLinearAccClamp=<2000 2000 2000>
    componentLinearAccScale=<0.2 0.2 0.2>
    componentLinearVelScale=<0.2 0.2 0.2>
    simSpaceSettings {
     maxAngularAcceleration=1000
     maxAngularVelocity=3
     maxLinearAcceleration=700
     maxLinearVelocity=500
     velocityScaleZ=0
     worldAlpha=0.1
     externalLinearDragV=<0.2 0.2>
    }
   }
   n6 = Component To Local(ComponentPose=@n5)
   Save cached pose 'Skirt Dynamics'(Pose=Blend Poses by bool(
     BlendPose_0=Use cached pose 'Skydive Cape '()
     BlendPose_1=Blend Poses by bool(BlendPose_0=@n6, BlendPose_1=@n3, bActiveValue=@n4) {blendType=Linear}
     bActiveValue=Get is Wearing Cape() -> (is Wearing Cape:bool)
    ) {blendType=Linear}) {cacheName=Skirt Dynamics}
   n7 = "Transform (Modify) Bone - Bone: attach_cape"(
    ComponentPose=Local To Component(LocalPose=Use cached pose 'Start Pose'())
    Alpha="Get is Skydiving "()
    Translation="0,-4,0"
    Rotation="0,0,0"
    Scale="1,1,1"
   ) {
    rotationMode=BMM_Additive
    rotationSpace=BCS_BoneSpace
    translationMode=BMM_Additive
    translationSpace=BCS_BoneSpace
    alphaBoolBlend {blendOption=Linear}
    alphaScaleBiasClamp {bInterpResult=true}
    boneToModify {boneName=attach_cape}
    translation {y=-4}
   }
   Save cached pose 'Skydive Cape '(Pose=Component To Local(ComponentPose=@n7)) {cacheName="Skydive Cape "}
   Use cached pose 'Front End Dynamics'()
  ]}
 EventGraph {nodes [
   Event BlueprintInitializeAnimation()
   > Set Skel Mesh(Skel Mesh=GetOwnerSkeletalMeshComponent())
   n1 = TryGetPawnOwner()
   > Is Valid(InputObject=@n1)
   > n2 = Cast To FortPlayerPawnAthena(Object=@n1)
   > n3 = Set Player Pawn(Player Pawn:Fort Player Pawn Athena Object Reference=@n2) -> (Output_Get:Fort Player Pawn Athena Object Reference)
   > Set Body Mesh(Body Mesh:Skeletal Mesh Component Object Reference=GetSkeletalMeshForPartType(self=@n3)) -> (Output_Get:Skeletal Mesh Component Object Reference)
   n4 = Event BlueprintUpdateAnimation()
   > n5 = Sequence!(execute) -> (then_0, then_1, then_2:Exec, then_3:Exec)
   "Set is Skydiving "(execute=@n5.then_2, "is Skydiving ":double=SelectFloat(A=0.35, B=1, bPickA=Get bIsSkyDiving())) -> (Output_Get:double)
   Is Valid(exec=@n5.then_3, InputObject=Get Player Pawn() -> (Player Pawn:Fort Player Pawn Athena Object Reference))
   > n6 = Branch(Condition=Get bIsInFrontEnd())
   Set WPO Globals Function(
    execute=@n6.else
    Material=Get Mat Instance Dynamic()
    Overlay Material=Get Overlay Mat Instance Dynamic()
    Skeletal Mesh=Get Skel Mesh()
    "Local Noise Transform  Location"=Get Local Noise Transform Location() -> (Local Noise Transform Location:Vector)
    Local Noise Offset=Get Local Noise Offset()
   )
   > Set WPO Regional Function!(
    execute
    Material:Material Instance Dynamic Object Reference=Get Mat Instance Dynamic()
    Overlay Material:Material Instance Dynamic Object Reference=Get Overlay Mat Instance Dynamic()
    Skeletal Mesh:Skeletal Mesh Component Object Reference=Get Skel Mesh()
    Velocity Transform:Name=foot_r
    Velocity Interp Speed:double=Get Velocity Interp Speed()
    Collision Transform:Name=calf_r
    Collision Transform Axis:EAxis Enum
    Invert Collision Transform Axis:bool=false
    Delta Time X:double=@n4.DeltaTimeX
    Previous Velocity Transform Location:"Vector (by ref)"=Get R1 Velocity Transform Location()
    Previous Velocity Transform Velocity:"Vector (by ref)"=Get R1 Velocity Transform Velocity()
    Region:int=1
   ) -> (then)
   > Set WPO Regional Function(
    Material=Get Mat Instance Dynamic()
    Overlay Material=Get Overlay Mat Instance Dynamic()
    Skeletal Mesh=Get Skel Mesh()
    Velocity Transform=foot_l
    Velocity Interp Speed=Get Velocity Interp Speed()
    Collision Transform=calf_l
    Invert Collision Transform Axis=true
    Delta Time X=@n4.DeltaTimeX
    Previous Velocity Transform Location=Get R2 Velocity Transform Location()
    Previous Velocity Transform Velocity=Get R2 Velocity Transform Velocity()
    Region=2
   )
   > n7 = Branch(Condition=OR Boolean(A=Get bIsSkydiveFloating(), B=Get bIsParachuteOpened()))
   > Set WPO Pho Wind Intensity(WPO Pho Wind Intensity=0.8)
   > n8 = Set WPO Pho Turbulance(WPO Pho Turbulance:double=1) -> (Output_Get:double)
   SetScalarParameterValue(
    execute=[@n8.then, @n10.then]
    self=[Get Mat Instance Dynamic(), Get Overlay Mat Instance Dynamic()]
    ParameterName=ClothSim_WindStrength
    Value=Get WPO Pho Wind Intensity()
   )
   n9 = Get WPO Pho Turbulance() -> (WPO Pho Turbulance:double)
   > SetScalarParameterValue(self=[Get Mat Instance Dynamic(), Get Overlay Mat Instance Dynamic()], ParameterName=ClothSim_R1_TurbulenceStrength, Value=@n9)
   > SetScalarParameterValue(self=[Get Mat Instance Dynamic(), Get Overlay Mat Instance Dynamic()], ParameterName=ClothSim_R2_TurbulenceStrength, Value=@n9)
   Set WPO Pho Wind Intensity(execute=@n7.else, WPO Pho Wind Intensity=0)
   > n10 = Set WPO Pho Turbulance(WPO Pho Turbulance=0.1)
   n11 = Gravity Override State Function DEPRECATED(
    execute=@n5.then_0
    CharacterPartAnimInstance=Self-Reference()
    GameTimeInSeconds=GetGameTimeInSeconds()
    GravityOverrideParams=Get Gravity Override Params()
   )
   > Set GravityOverride(GravityOverride=@n11)
   n12 = CheckForBackbling()
   Set is Wearing Backbling(
    execute=@n5.then_1
    is Wearing Backbling:bool=AND Boolean!(A=@n12, B=NOT Boolean(A=@n12.bIsBackblingACape), C:bool=NOT Boolean(A=@n12.bIsBackblingLong)) -> (ReturnValue)
   ) -> (Output_Get:bool)
   > Set is Wearing Cape(is Wearing Cape:bool=OR Boolean(A=@n12.bIsBackblingACape, B=@n12.bIsBackblingLong)) -> (Output_Get:bool)
   Event OnCharacterCustomizationLoaded()
   > n13 = Set Skel Mesh(Skel Mesh=GetPartSkeletalMeshComponent())
   > n14 = Get Dynamic Material Instance(SkeletalMeshComponent=@n13)
   > Set Mat Instance Dynamic(Mat Instance Dynamic=@n14)
   > n15 = GetOrCreateDynamicOverlayMaterial(MeshComponent=@n13, MaterialName=None)
   > Set Overlay Mat Instance Dynamic(Overlay Mat Instance Dynamic=@n15)
   TryGetPawnOwner()
  ]}
]
variables [
 Body Mesh {type=SkeletalMeshComponent}
 Gravity Override Params {type=GravityOverrideParamsStruct}
 GravityOverride {type=Vector}
 "is Skydiving " {type=number}
 is Wearing Backbling {type=boolean}
 is Wearing Cape {type=boolean}
 Local Noise Offset {type=Vector}
 Local Noise Transform Location {type=Vector}
 Mat Instance Dynamic {type=MaterialInstanceDynamic}
 Overlay Mat Instance Dynamic {type=MaterialInstanceDynamic}
 Player Pawn {type=FortPlayerPawnAthena}
 R1 Velocity Transform Location {type=Vector}
 R1 Velocity Transform Velocity {type=Vector}
 R2 Velocity Transform Location {type=Vector}
 R2 Velocity Transform Velocity {type=Vector}
 Skel Mesh {type=SkeletalMeshComponent}
 Velocity Interp Speed {type=number}
 WPO Pho Turbulance {type=number}
 WPO Pho Wind Intensity {type=number}
]
