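# Physics asset for the Ballet Assassin (female, medium) FaceAcc_Backpack mesh part, providing bodies and joints for hair, hair tail and ribbon motion.
# Capsules cover head, neck and each chain, with dyn_ bones linked in sequence from small base_ anchors on the head.
# Hair and ribbon joints restrict swing and tightly limit twist, loosening along the long ribbons.
# Hair tail, anchor-to-head and head-to-neck joints leave swing and twist free.
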PhysicsAsset /BRCosmetics/Characters/Player/Female/Medium/Bodies/F_MED_BalletAssassin/Meshes/Parts/F_MED_BalletAssassin_FaceAcc_Backpack_PhysicsAsset
bodies {
 base_hairTail_00 [{shapeType=Capsule, length=0.505, radius=0.505, shapeName=base_hairTail_00_capsule, rotation=<rot 0 -90 0>}]
 base_hair_0 [{shapeType=Capsule, length=0.505, radius=0.505, shapeName=base_hair_0_capsule, rotation=<rot 0 -90 0>}]
 base_hair_00_l [{shapeType=Capsule, length=0.505, radius=0.505, shapeName=base_hair_00_l_capsule, rotation=<rot 0 -90 0>}]
 base_hair_00_r [{shapeType=Capsule, length=0.505, radius=0.505, shapeName=base_hair_00_r_capsule, rotation=<rot 0 -90 0>}]
 base_ribbonTop_l [{shapeType=Capsule, length=0.505, radius=0.505, shapeName=base_ribbonTop_l_capsule, rotation=<rot 0 -90 0>}]
 base_ribbonTop_r [{shapeType=Capsule, length=0.505, radius=0.505, shapeName=base_ribbonTop_r_capsule, rotation=<rot 0 -90 0>}]
 base_ribbon_00_l [{shapeType=Capsule, length=0.505, radius=0.505, shapeName=base_ribbon_00_l_capsule, rotation=<rot 0 -90 0>}]
 base_ribbon_00_r [{shapeType=Capsule, length=0.505, radius=0.505, shapeName=base_ribbon_00_r_capsule, rotation=<rot 0 -90 0>}]
 dyn_hairTail_00 [{shapeType=Capsule, length=2.8032, radius=4.4274, shapeName=dyn_hairTail_00_capsule, center=<3.6802 1.5558 0>, rotation=<rot 90 0 0>}]
 dyn_hairTail_01 [{shapeType=Capsule, length=5.214, radius=5.3429, shapeName=dyn_hairTail_01_capsule, center=<4.0649 -1.7094 -0.8603>, rotation=<rot 90 0 0>}]
 dyn_hairTail_02 [{shapeType=Capsule, length=6.1294, radius=4.8644, shapeName=dyn_hairTail_02_capsule, center=<5.2 -0.4074 -1.3518>, rotation=<rot 90 0 0>}]
 dyn_hairTail_03 [{shapeType=Capsule, length=6.0207, radius=4.9694, shapeName=dyn_hairTail_03_capsule, center=<3.0911 1.243 0>, rotation=<rot 90 0 -8.3132>}]
 dyn_hair_00 [{shapeType=Capsule, length=2, radius=1, shapeName=dyn_hair_00_capsule, center=<2 0 0>, rotation=<rot 90 0 0>}]
 dyn_hair_00_l [{shapeType=Capsule, length=2, radius=1, shapeName=dyn_hair_00_l_capsule, center=<2 0 0>, rotation=<rot 90 0 0>}]
 dyn_hair_00_r [{shapeType=Capsule, length=2, radius=1, shapeName=dyn_hair_00_r_capsule, center=<2 0 0>, rotation=<rot 90 0 0>}]
 dyn_hair_01 [{shapeType=Capsule, length=2, radius=1, shapeName=dyn_hair_01_capsule, center=<2 0 0>, rotation=<rot 90 0 0>}]
 dyn_hair_01_l [{shapeType=Capsule, length=2, radius=1, shapeName=dyn_hair_01_l_capsule, center=<2 0 0>, rotation=<rot 90 0 0>}]
 dyn_hair_01_r [{shapeType=Capsule, length=2, radius=1, shapeName=dyn_hair_01_r_capsule, center=<2 0 0>, rotation=<rot 90 0 0>}]
 dyn_ribbonTop_l [{shapeType=Capsule, length=2, radius=1, shapeName=dyn_ribbonTop_l_capsule, center=<2 0 0>, rotation=<rot 90 0 0>}]
 dyn_ribbonTop_r [{shapeType=Capsule, length=2, radius=1, shapeName=dyn_ribbonTop_r_capsule, center=<2 0 0>, rotation=<rot 90 0 0>}]
 dyn_ribbon_00_l [{shapeType=Capsule, length=2, radius=1, shapeName=dyn_ribbon_00_l_capsule, center=<2 0 0>, rotation=<rot 90 0 0>}]
 dyn_ribbon_00_r [{shapeType=Capsule, length=2, radius=1, shapeName=dyn_ribbon_00_r_capsule, center=<2 0 0>, rotation=<rot 90 0 0>}]
 dyn_ribbon_01_l [{shapeType=Capsule, length=3.2935, radius=1, shapeName=dyn_ribbon_01_l_capsule, center=<2.6823 -0.8171 0>, rotation=<rot 90 0 0>}]
 dyn_ribbon_01_r [{shapeType=Capsule, length=2, radius=1, shapeName=dyn_ribbon_01_r_capsule, center=<2 -0.7484 0>, rotation=<rot 90 0 0>}]
 dyn_ribbon_02_l [{shapeType=Capsule, length=5.0016, radius=1, shapeName=dyn_ribbon_02_l_capsule, center=<2.2638 -0.595 -0.2337>, rotation=<rot 90 0 0>}]
 dyn_ribbon_02_r [{shapeType=Capsule, length=5.0149, radius=1, shapeName=dyn_ribbon_02_r_capsule, center=<2.4468 -0.662 0>, rotation=<rot 90 0 0>}]
 dyn_ribbon_03_l [{shapeType=Capsule, length=8.303, radius=1, shapeName=dyn_ribbon_03_l_capsule, center=<3.603 -0.4966 0>, rotation=<rot 90 0 0>}]
 head [{shapeType=Capsule, length=2.3784, radius=7.2578, shapeName=head_capsule, center=<6.0939 -0.2551 0>, rotation=<rot 90 0 0>}]
 neck_01 [{shapeType=Capsule, length=7.9965, radius=4.265, shapeName=neck_01_capsule, center=<4.265 0 0>, rotation=<rot 90 0 0>}]
}
constraints [
 {bone1Name=dyn_hair_00, bone2Name=base_hair_0, swing1LimitDegrees=15, swing2LimitDegrees=10, twistLimitDegrees=1}
 {bone1Name=dyn_hair_01, bone2Name=dyn_hair_00, swing1LimitDegrees=15, swing2LimitDegrees=10, twistLimitDegrees=1}
 {bone1Name=dyn_hair_00_l, bone2Name=base_hair_00_l, swing1LimitDegrees=15, swing2LimitDegrees=20, twistLimitDegrees=1}
 {bone1Name=dyn_hair_01_l, bone2Name=dyn_hair_00_l, swing1LimitDegrees=15, swing2LimitDegrees=20, twistLimitDegrees=1}
 {bone1Name=dyn_hair_00_r, bone2Name=base_hair_00_r, swing1LimitDegrees=15, swing2LimitDegrees=25, twistLimitDegrees=1}
 {bone1Name=dyn_hair_01_r, bone2Name=dyn_hair_00_r, swing1LimitDegrees=15, swing2LimitDegrees=20, twistLimitDegrees=1}
 {bone1Name=dyn_ribbon_00_l, bone2Name=base_ribbon_00_l, swing1LimitDegrees=10, swing2LimitDegrees=10, twistLimitDegrees=1}
 {bone1Name=dyn_ribbon_01_l, bone2Name=dyn_ribbon_00_l, swing1LimitDegrees=30, swing2LimitDegrees=30, twistLimitDegrees=1}
 {bone1Name=dyn_ribbon_02_l, bone2Name=dyn_ribbon_01_l, swing1LimitDegrees=35, swing2LimitDegrees=35, twistLimitDegrees=1}
 {bone1Name=dyn_ribbon_03_l, bone2Name=dyn_ribbon_02_l, swing1LimitDegrees=35, swing2LimitDegrees=35, twistLimitDegrees=1}
 {bone1Name=dyn_ribbon_00_r, bone2Name=base_ribbon_00_r, swing1LimitDegrees=10, swing2LimitDegrees=10, twistLimitDegrees=1}
 {bone1Name=dyn_ribbon_01_r, bone2Name=dyn_ribbon_00_r, swing1LimitDegrees=25, swing2LimitDegrees=25, twistLimitDegrees=1}
 {bone1Name=dyn_ribbon_02_r, bone2Name=dyn_ribbon_01_r, swing1LimitDegrees=35, swing2LimitDegrees=35, twistLimitDegrees=1}
 {
  bone1Name=base_hair_00_l
  bone2Name=head
  swing1LimitDegrees=45
  swing1Motion=Free
  swing2LimitDegrees=45
  swing2Motion=Free
  twistLimitDegrees=45
  twistMotion=Free
 }
 {
  bone1Name=base_hair_00_r
  bone2Name=head
  swing1LimitDegrees=45
  swing1Motion=Free
  swing2LimitDegrees=45
  swing2Motion=Free
  twistLimitDegrees=45
  twistMotion=Free
 }
 {
  bone1Name=base_hair_0
  bone2Name=head
  swing1LimitDegrees=45
  swing1Motion=Free
  swing2LimitDegrees=45
  swing2Motion=Free
  twistLimitDegrees=45
  twistMotion=Free
 }
 {
  bone1Name=base_ribbon_00_l
  bone2Name=head
  swing1LimitDegrees=45
  swing1Motion=Free
  swing2LimitDegrees=45
  swing2Motion=Free
  twistLimitDegrees=45
  twistMotion=Free
 }
 {
  bone1Name=base_ribbon_00_r
  bone2Name=head
  swing1LimitDegrees=45
  swing1Motion=Free
  swing2LimitDegrees=45
  swing2Motion=Free
  twistLimitDegrees=45
  twistMotion=Free
 }
 {
  bone1Name=base_hairTail_00
  bone2Name=head
  swing1LimitDegrees=45
  swing1Motion=Free
  swing2LimitDegrees=45
  swing2Motion=Free
  twistLimitDegrees=45
  twistMotion=Free
 }
 {
  bone1Name=dyn_hairTail_00
  bone2Name=base_hairTail_00
  swing1LimitDegrees=45
  swing1Motion=Free
  swing2LimitDegrees=45
  swing2Motion=Free
  twistLimitDegrees=45
  twistMotion=Free
 }
 {
  bone1Name=dyn_hairTail_01
  bone2Name=dyn_hairTail_00
  swing1LimitDegrees=45
  swing1Motion=Free
  swing2LimitDegrees=45
  swing2Motion=Free
  twistLimitDegrees=45
  twistMotion=Free
 }
 {
  bone1Name=dyn_hairTail_02
  bone2Name=dyn_hairTail_01
  swing1LimitDegrees=45
  swing1Motion=Free
  swing2LimitDegrees=45
  swing2Motion=Free
  twistLimitDegrees=45
  twistMotion=Free
 }
 {
  bone1Name=dyn_hairTail_03
  bone2Name=dyn_hairTail_02
  swing1LimitDegrees=45
  swing1Motion=Free
  swing2LimitDegrees=45
  swing2Motion=Free
  twistLimitDegrees=45
  twistMotion=Free
 }
 {
  bone1Name=base_ribbonTop_l
  bone2Name=head
  swing1LimitDegrees=45
  swing1Motion=Free
  swing2LimitDegrees=45
  swing2Motion=Free
  twistLimitDegrees=45
  twistMotion=Free
 }
 {bone1Name=dyn_ribbonTop_l, bone2Name=base_ribbonTop_l, swing1LimitDegrees=10, swing2LimitDegrees=1, twistLimitDegrees=1}
 {
  bone1Name=base_ribbonTop_r
  bone2Name=head
  swing1LimitDegrees=45
  swing1Motion=Free
  swing2LimitDegrees=45
  swing2Motion=Free
  twistLimitDegrees=45
  twistMotion=Free
 }
 {bone1Name=dyn_ribbonTop_r, bone2Name=base_ribbonTop_r, swing1LimitDegrees=10, swing2LimitDegrees=1, twistLimitDegrees=1}
 {
  bone1Name=head
  bone2Name=neck_01
  swing1LimitDegrees=45
  swing1Motion=Free
  swing2LimitDegrees=45
  swing2Motion=Free
  twistLimitDegrees=45
  twistMotion=Free
 }
]
