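# Physics asset for the female BalletAssassin face accessory mesh, setting up capsule bodies and joints for a hair tail, hair strands and ribbons.
# Small base bodies join each piece to the head, and chained dyn_ bodies hang from them with limited swing and minimal or locked twist.
# Larger capsules on pelvis, spine, neck, head and upper arms form a body proxy, linked by angularly free joints.
# The left ribbon chain has one more segment than the right.
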PhysicsAsset /BRCosmetics/Characters/Player/Female/Medium/Bodies/F_MED_BalletAssassin/Meshes/Parts/F_MED_BalletAssassin_FaceAcc_PhysicsAsset
bodies {
 base_hairTail_00 [{shapeType=Capsule, length=0.505, radius=0.505, shapeName=base_hairTail_00_capsule, rotation=<rot 0 -90 0>}]
 base_hair_0 [{shapeType=Capsule, length=0.505, radius=0.505, shapeName=base_hair_0_capsule, rotation=<rot 0 -90 0>}]
 base_hair_00_l [{shapeType=Capsule, length=0.505, radius=0.505, shapeName=base_hair_00_l_capsule, rotation=<rot 0 -90 0>}]
 base_hair_00_r [{shapeType=Capsule, length=0.505, radius=0.505, shapeName=base_hair_00_r_capsule, rotation=<rot 0 -90 0>}]
 base_ribbonTop_l [{shapeType=Capsule, length=0.505, radius=0.505, shapeName=base_ribbonTop_l_capsule, rotation=<rot 0 -90 0>}]
 base_ribbonTop_r [{shapeType=Capsule, length=0.505, radius=0.505, shapeName=base_ribbonTop_r_capsule, rotation=<rot 0 -90 0>}]
 base_ribbon_00_l [{shapeType=Capsule, length=0.505, radius=0.505, shapeName=base_ribbon_00_l_capsule, rotation=<rot 0 -90 0>}]
 base_ribbon_00_r [{shapeType=Capsule, length=0.505, radius=0.505, shapeName=base_ribbon_00_r_capsule, rotation=<rot 0 -90 0>}]
 dyn_hairTail_00 [{shapeType=Capsule, length=1.5581, radius=1.4419, shapeName=dyn_hairTail_00_capsule, center=<2 0 0>, rotation=<rot 90 0 0>}]
 dyn_hairTail_01 [{shapeType=Capsule, length=4.298, radius=4.3921, shapeName=dyn_hairTail_01_capsule, center=<4.6503 -0.9967 -0.9797>, rotation=<rot 90 0 0>}]
 dyn_hairTail_02 [{shapeType=Capsule, length=8.1923, radius=5.5466, shapeName=dyn_hairTail_02_capsule, center=<4.4453 0 -1.1528>, rotation=<rot 90 0 0>}]
 dyn_hairTail_03 [{shapeType=Capsule, length=5.6636, radius=6.2202, shapeName=dyn_hairTail_03_capsule, center=<5.4351 -0.2767 0>, rotation=<rot 90 0 0>}]
 dyn_hairTail_04 [{
   shapeType=Capsule
   length=6.1707
   radius=6.2909
   shapeName=dyn_hairTail_04_capsule
   center=<7.4532 -0.3206 0>
   rotation=<rot 90 -178.2101 -167.524>
  }]
 dyn_hairTail_05 [{
   shapeType=Capsule
   length=11.4283
   radius=5.8124
   shapeName=dyn_hairTail_05_capsule
   center=<9.0313 -1.2874 -1.0289>
   rotation=<rot 90 180 176.2275>
  }]
 dyn_hair_00 [{shapeType=Capsule, length=2, radius=1, shapeName=dyn_hair_00_capsule, center=<2 0 0>, rotation=<rot 90 0 0>}]
 dyn_hair_00_l [{shapeType=Capsule, length=2, radius=1, shapeName=dyn_hair_00_l_capsule, center=<2 0 0>, rotation=<rot 90 0 0>}]
 dyn_hair_00_r [{shapeType=Capsule, length=2, radius=1, shapeName=dyn_hair_00_r_capsule, center=<2 0 0>, rotation=<rot 90 0 0>}]
 dyn_hair_01 [{shapeType=Capsule, length=2, radius=1, shapeName=dyn_hair_01_capsule, center=<2 0 0>, rotation=<rot 90 0 0>}]
 dyn_hair_01_l [{shapeType=Capsule, length=2, radius=1, shapeName=dyn_hair_01_l_capsule, center=<2 0 0>, rotation=<rot 90 0 0>}]
 dyn_hair_01_r [{shapeType=Capsule, length=2, radius=1, shapeName=dyn_hair_01_r_capsule, center=<2 0 0>, rotation=<rot 90 0 0>}]
 dyn_ribbonTop_l [{shapeType=Capsule, length=2, radius=1, shapeName=dyn_ribbonTop_l_capsule, center=<2 0 0>, rotation=<rot 90 0 0>}]
 dyn_ribbonTop_r [{shapeType=Capsule, length=2, radius=1, shapeName=dyn_ribbonTop_r_capsule, center=<2 0 0>, rotation=<rot 90 0 0>}]
 dyn_ribbon_00_l [{shapeType=Capsule, length=2, radius=1, shapeName=dyn_ribbon_00_l_capsule, center=<2 0 0>, rotation=<rot 90 0 0>}]
 dyn_ribbon_00_r [{shapeType=Capsule, length=2, radius=1, shapeName=dyn_ribbon_00_r_capsule, center=<2 0 0>, rotation=<rot 90 0 0>}]
 dyn_ribbon_01_l [{shapeType=Capsule, length=2.6297, radius=1, shapeName=dyn_ribbon_01_l_capsule, center=<2.3741 -0.8171 0>, rotation=<rot 90 0 0>}]
 dyn_ribbon_01_r [{shapeType=Capsule, length=2, radius=1, shapeName=dyn_ribbon_01_r_capsule, center=<2 -0.7484 0>, rotation=<rot 90 0 0>}]
 dyn_ribbon_02_l [{shapeType=Capsule, length=5.0016, radius=1, shapeName=dyn_ribbon_02_l_capsule, center=<2.2638 -0.595 -0.2337>, rotation=<rot 90 0 0>}]
 dyn_ribbon_02_r [{shapeType=Capsule, length=5.0149, radius=1, shapeName=dyn_ribbon_02_r_capsule, center=<2.4468 -0.662 0>, rotation=<rot 90 0 0>}]
 dyn_ribbon_03_l [{shapeType=Capsule, length=7.3608, radius=1.9422, shapeName=dyn_ribbon_03_l_capsule, center=<3.603 -1.4974 -0.3832>, rotation=<rot 90 0 0>}]
 head [{shapeType=Capsule, length=2.3784, radius=7.2578, shapeName=head_capsule, center=<6.0939 -0.2551 0>, rotation=<rot 90 0 0>}]
 neck_01 [{shapeType=Capsule, length=7.9965, radius=4.265, shapeName=neck_01_capsule, center=<5.6733 0 0>, rotation=<rot 90 0 0>}]
 pelvis [{shapeType=Capsule, length=12.517, radius=13.3151, shapeName=pelvis_capsule, center=<2.9804 1.2433 0>}]
 spine_01 [{shapeType=Capsule, length=12.1999, radius=12.9265, shapeName=spine_01_capsule, center=<8.7088 4.4634 0>, rotation=<rot 0 -90 0>}]
 spine_02 [{shapeType=Capsule, length=17.2132, radius=10.6397, shapeName=spine_02_capsule, center=<11.639 3.8226 0>}]
 spine_04 [{shapeType=Capsule, length=16.7792, radius=11.1537, shapeName=spine_04_capsule, center=<9.4659 3.4723 0>}]
 spine_05 [{shapeType=Capsule, length=15.6807, radius=13.2289, shapeName=spine_05_capsule, center=<7.9109 7.351 0>, rotation=<rot 0 4.7836 0>}]
 upperarm_l [{shapeType=Capsule, length=22.238, radius=6.2708, shapeName=upperarm_l_capsule, center=<11.9708 1.0875 0>, rotation=<rot 90 0 0>}]
 upperarm_r [{shapeType=Capsule, length=22.238, radius=6.2708, shapeName=upperarm_r_capsule, center=<-11.8601 -0.9596 0>, rotation=<rot 90 0 0>}]
}
constraints [
 {bone1Name=dyn_hair_00, bone2Name=base_hair_0, swing1LimitDegrees=15, swing2LimitDegrees=10, twistLimitDegrees=1}
 {bone1Name=dyn_hair_01, bone2Name=dyn_hair_00, swing1LimitDegrees=15, swing2LimitDegrees=10, twistLimitDegrees=1}
 {bone1Name=dyn_hair_00_l, bone2Name=base_hair_00_l, swing1LimitDegrees=15, swing2LimitDegrees=20, twistLimitDegrees=1}
 {bone1Name=dyn_hair_01_l, bone2Name=dyn_hair_00_l, swing1LimitDegrees=15, swing2LimitDegrees=20, twistLimitDegrees=1}
 {bone1Name=dyn_hair_00_r, bone2Name=base_hair_00_r, swing1LimitDegrees=15, swing2LimitDegrees=25, twistLimitDegrees=1}
 {bone1Name=dyn_hair_01_r, bone2Name=dyn_hair_00_r, swing1LimitDegrees=15, swing2LimitDegrees=20, twistLimitDegrees=1}
 {bone1Name=dyn_hairTail_00, bone2Name=base_hairTail_00, swing1LimitDegrees=15, swing2LimitDegrees=15, twistLimitDegrees=1}
 {bone1Name=dyn_hairTail_01, bone2Name=dyn_hairTail_00, swing1LimitDegrees=40, swing2LimitDegrees=12, twistLimitDegrees=1}
 {bone1Name=dyn_hairTail_02, bone2Name=dyn_hairTail_01, swing1LimitDegrees=45, swing2LimitDegrees=11, twistLimitDegrees=1}
 {bone1Name=dyn_hairTail_03, bone2Name=dyn_hairTail_02, swing1LimitDegrees=40, swing2LimitDegrees=11, twistLimitDegrees=1}
 {bone1Name=dyn_hairTail_04, bone2Name=dyn_hairTail_03, swing1LimitDegrees=40, swing2LimitDegrees=11, twistLimitDegrees=1}
 {bone1Name=dyn_hairTail_05, bone2Name=dyn_hairTail_04, swing1LimitDegrees=40, swing2LimitDegrees=10, twistLimitDegrees=1}
 {bone1Name=dyn_ribbon_00_l, bone2Name=base_ribbon_00_l, swing1LimitDegrees=10, swing2LimitDegrees=10, twistLimitDegrees=1}
 {bone1Name=dyn_ribbon_01_l, bone2Name=dyn_ribbon_00_l, swing1LimitDegrees=30, swing2LimitDegrees=30, twistLimitDegrees=1}
 {bone1Name=dyn_ribbon_02_l, bone2Name=dyn_ribbon_01_l, swing1LimitDegrees=35, swing2LimitDegrees=35, twistLimitDegrees=1}
 {bone1Name=dyn_ribbon_03_l, bone2Name=dyn_ribbon_02_l, swing1LimitDegrees=35, swing2LimitDegrees=35, twistLimitDegrees=1}
 {bone1Name=dyn_ribbon_00_r, bone2Name=base_ribbon_00_r, swing1LimitDegrees=10, swing2LimitDegrees=10, twistLimitDegrees=1}
 {bone1Name=dyn_ribbon_01_r, bone2Name=dyn_ribbon_00_r, swing1LimitDegrees=25, swing2LimitDegrees=25, twistLimitDegrees=1}
 {bone1Name=dyn_ribbon_02_r, bone2Name=dyn_ribbon_01_r, swing1LimitDegrees=35, swing2LimitDegrees=35, twistLimitDegrees=1, twistMotion=Locked}
 {
  bone1Name=base_hair_00_l
  bone2Name=head
  swing1LimitDegrees=45
  swing1Motion=Free
  swing2LimitDegrees=45
  swing2Motion=Free
  twistLimitDegrees=45
  twistMotion=Free
 }
 {
  bone1Name=base_hair_00_r
  bone2Name=head
  swing1LimitDegrees=45
  swing1Motion=Free
  swing2LimitDegrees=45
  swing2Motion=Free
  twistLimitDegrees=45
  twistMotion=Free
 }
 {
  bone1Name=base_hair_0
  bone2Name=head
  swing1LimitDegrees=45
  swing1Motion=Free
  swing2LimitDegrees=45
  swing2Motion=Free
  twistLimitDegrees=45
  twistMotion=Free
 }
 {
  bone1Name=base_ribbon_00_l
  bone2Name=head
  swing1LimitDegrees=45
  swing1Motion=Free
  swing2LimitDegrees=45
  swing2Motion=Free
  twistLimitDegrees=45
  twistMotion=Free
 }
 {
  bone1Name=base_ribbon_00_r
  bone2Name=head
  swing1LimitDegrees=45
  swing1Motion=Free
  swing2LimitDegrees=45
  swing2Motion=Free
  twistLimitDegrees=45
  twistMotion=Free
 }
 {
  bone1Name=base_hairTail_00
  bone2Name=head
  swing1LimitDegrees=45
  swing1Motion=Free
  swing2LimitDegrees=45
  swing2Motion=Free
  twistLimitDegrees=45
  twistMotion=Free
 }
 {
  bone1Name=spine_05
  bone2Name=spine_04
  swing1LimitDegrees=45
  swing1Motion=Free
  swing2LimitDegrees=45
  swing2Motion=Free
  twistLimitDegrees=45
  twistMotion=Free
 }
 {
  bone1Name=head
  bone2Name=neck_01
  swing1LimitDegrees=45
  swing1Motion=Free
  swing2LimitDegrees=45
  swing2Motion=Free
  twistLimitDegrees=45
  twistMotion=Free
 }
 {
  bone1Name=upperarm_l
  bone2Name=spine_05
  swing1LimitDegrees=45
  swing1Motion=Free
  swing2LimitDegrees=45
  swing2Motion=Free
  twistLimitDegrees=45
  twistMotion=Free
 }
 {
  bone1Name=upperarm_r
  bone2Name=spine_05
  swing1LimitDegrees=45
  swing1Motion=Free
  swing2LimitDegrees=45
  swing2Motion=Free
  twistLimitDegrees=45
  twistMotion=Free
 }
 {
  bone1Name=spine_02
  bone2Name=spine_01
  swing1LimitDegrees=45
  swing1Motion=Free
  swing2LimitDegrees=45
  swing2Motion=Free
  twistLimitDegrees=45
  twistMotion=Free
 }
 {
  bone1Name=base_ribbonTop_l
  bone2Name=head
  swing1LimitDegrees=45
  swing1Motion=Free
  swing2LimitDegrees=45
  swing2Motion=Free
  twistLimitDegrees=45
  twistMotion=Free
 }
 {bone1Name=dyn_ribbonTop_l, bone2Name=base_ribbonTop_l, swing1LimitDegrees=10, swing2LimitDegrees=1, twistLimitDegrees=1}
 {
  bone1Name=base_ribbonTop_r
  bone2Name=head
  swing1LimitDegrees=45
  swing1Motion=Free
  swing2LimitDegrees=45
  swing2Motion=Free
  twistLimitDegrees=45
  twistMotion=Free
 }
 {bone1Name=dyn_ribbonTop_r, bone2Name=base_ribbonTop_r, swing1LimitDegrees=10, swing2LimitDegrees=1, twistLimitDegrees=1}
 {
  bone1Name=spine_01
  bone2Name=pelvis
  swing1LimitDegrees=45
  swing1Motion=Free
  swing2LimitDegrees=45
  swing2Motion=Free
  twistLimitDegrees=45
  twistMotion=Free
 }
 {
  bone1Name=spine_04
  bone2Name=spine_02
  swing1LimitDegrees=45
  swing1Motion=Free
  swing2LimitDegrees=45
  swing2Motion=Free
  twistLimitDegrees=45
  twistMotion=Free
 }
 {
  bone1Name=neck_01
  bone2Name=spine_05
  swing1LimitDegrees=45
  swing1Motion=Free
  swing2LimitDegrees=45
  swing2Motion=Free
  twistLimitDegrees=45
  twistMotion=Free
 }
]
