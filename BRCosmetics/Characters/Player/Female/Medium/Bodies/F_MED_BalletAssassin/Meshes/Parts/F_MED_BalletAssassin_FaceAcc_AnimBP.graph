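# Animation Blueprint for the BalletAssassin face accessory mesh, animating its hair tail on the player character.
# It copies the parent mesh pose, then chooses cape, backpack or normal dynamics from worn back bling, using separate RigidBody physics assets.
# Normal falls back to bone constraints and head-tilt correctives without rigid bodies; backpack layers physics over them.
# The event graph updates gravity override, back-bling flags, head angles and movement alphas driving these adjustments.
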AnimBlueprint /BRCosmetics/Characters/Player/Female/Medium/Bodies/F_MED_BalletAssassin/Meshes/Parts/F_MED_BalletAssassin_FaceAcc_AnimBP
target_skeleton=&/F_MED_BalletAssassin_FaceAcc_Skeleton
graphs [
 AnimGraph {nodes [
   Local To Component()
   Local To Component()
   n1 = Use cached pose 'Backpack Dynamics'()
   n2 = Use cached pose 'Normal Dynamics'()
   n3 = Get is Wearing Backbling() -> (is Wearing Backbling:bool)
   Output Pose(Result=Blend Poses by bool(
     BlendPose_0=Use cached pose 'Cape Dynamics'()
     BlendPose_1=Blend Poses by bool(BlendPose_0=@n1, BlendPose_1=@n2, bActiveValue=@n3) {blendType=Linear}
     bActiveValue=Get is Wearing Cape() -> (is Wearing Cape:bool)
    ) {blendType=Linear}) {layerGroup=None}
   Save cached pose 'Start Pose'(Pose=Copy Pose From Mesh(SourceMeshComponent=Get Skel Mesh()) {bUseAttachedParent=true}) {cacheName=Start Pose}
   n4 = "Constraint - dyn_hairTail_00 (PP)"!(ComponentPose=Local To Component(LocalPose=Use cached pose 'Start Pose'()), Alpha=1, ConstraintWeights_0=0.9, ConstraintWeights_1:float=0.1) -> (Pose) {
    alphaBoolBlend {blendOption=Linear}
    boneToModify {boneName=dyn_hairTail_00}
    constraintSetup [{targetBone {boneName=head}}, {targetBone {boneName=spine_05}}]
    constraintWeights [1,1]
   }
   n5 = "Constraint - dyn_hairTail_01 (PP)"!(ComponentPose=@n4, Alpha=1, ConstraintWeights_0=0.8, ConstraintWeights_1:float=0.2) -> (Pose) {
    alphaBoolBlend {blendOption=Linear}
    boneToModify {boneName=dyn_hairTail_01}
    constraintSetup [{targetBone {boneName=head}}, {targetBone {boneName=spine_05}}]
    constraintWeights [1,1]
   }
   n6 = "Constraint - dyn_hairTail_02 (PP)"!(ComponentPose=@n5, Alpha=1, ConstraintWeights_0=0.6, ConstraintWeights_1:float=0.4) -> (Pose) {
    alphaBoolBlend {blendOption=Linear}
    boneToModify {boneName=dyn_hairTail_02}
    constraintSetup [{targetBone {boneName=head}}, {targetBone {boneName=spine_05}}]
    constraintWeights [1,1]
   }
   n7 = "Constraint - dyn_hairTail_03 (PP)"!(ComponentPose=@n6, Alpha=1, ConstraintWeights_0=0.3, ConstraintWeights_1:float=0.7) -> (Pose) {
    alphaBoolBlend {blendOption=Linear}
    boneToModify {boneName=dyn_hairTail_03}
    constraintSetup [{targetBone {boneName=head}}, {targetBone {boneName=spine_05}}]
    constraintWeights [1,1]
   }
   n8 = "Constraint - dyn_hairTail_04 (PPP)"!(ComponentPose=@n7, Alpha=1, ConstraintWeights_0=0.3, ConstraintWeights_1:float=0.5, ConstraintWeights_2:float=0.2) -> (Pose) {
    alphaBoolBlend {blendOption=Linear}
    boneToModify {boneName=dyn_hairTail_04}
    constraintSetup [{targetBone {boneName=head}}, {targetBone {boneName=spine_05}}, {targetBone {boneName=spine_04}}]
    constraintWeights [1,1,1]
   }
   n9 = "Constraint - dyn_hairTail_05 (PPP)"!(ComponentPose=@n8, Alpha=1, ConstraintWeights_0=0.7, ConstraintWeights_1:float=0, ConstraintWeights_2:float=0.3) -> (Pose) {
    alphaBoolBlend {blendOption=Linear}
    boneToModify {boneName=dyn_hairTail_05}
    constraintSetup [{targetBone {boneName=spine_05}}, {targetBone {boneName=spine_04}}, {targetBone {boneName=head}}]
    constraintWeights [1,1,1]
   }
   n10 = Get head_fwdBwd_rot() -> (head_fwdBwd_rot:double)
   n11 = "Transform (Modify) Bone - Bone: dyn_hairTail_01"(ComponentPose=@n9, Alpha=@n10, Translation="0.015,-1,0.678", Rotation="0.002,-5,0.001", Scale="1,1,1") {
    rotationMode=BMM_Additive
    rotationSpace=BCS_BoneSpace
    translationMode=BMM_Additive
    translationSpace=BCS_BoneSpace
    alphaBoolBlend {blendOption=Linear}
    alphaScaleBiasClamp {bMapRange=true, inRange {max=0, min=-25}, outRange {max=0, min=1}}
    boneToModify {boneName=dyn_hairTail_01}
    rotation=<rot 0.002 -5 0.001>
    translation=<0.015 -1 0.678>
   }
   n12 = "Transform (Modify) Bone - Bone: dyn_hairTail_02"(ComponentPose=@n11, Alpha=@n10, Translation="-0.001,-0.467,-0.240", Rotation="-0.010,-2.865,0.003", Scale="1,1,1") {
    rotationMode=BMM_Additive
    rotationSpace=BCS_BoneSpace
    translationMode=BMM_Additive
    translationSpace=BCS_BoneSpace
    alphaBoolBlend {blendOption=Linear}
    alphaScaleBiasClamp {bMapRange=true, inRange {max=0, min=-25}, outRange {max=0, min=1}}
    boneToModify {boneName=dyn_hairTail_02}
    rotation=<rot -0.01 -2.865 0.003>
    translation=<-0.001 -0.467 -0.24>
   }
   n13 = Get head_fwdBwd_rot()
   n14 = "Transform (Modify) Bone - Bone: dyn_hairTail_01"(ComponentPose=@n12, Alpha=@n13, Translation="0,1,0", Rotation="0,0,0", Scale="1,1,1") {
    rotationMode=BMM_Additive
    rotationSpace=BCS_BoneSpace
    translationMode=BMM_Additive
    translationSpace=BCS_BoneSpace
    alphaBoolBlend {blendOption=Linear}
    alphaScaleBiasClamp {bMapRange=true, inRange {max=15}}
    boneToModify {boneName=dyn_hairTail_01}
    translation {y=1}
   }
   n15 = "Transform (Modify) Bone - Bone: dyn_hairTail_02"(ComponentPose=@n14, Alpha=@n13, Translation="0.009,1,0", Rotation="0,0,0", Scale="1,1,1") {
    rotationMode=BMM_Additive
    rotationSpace=BCS_BoneSpace
    translationMode=BMM_Additive
    translationSpace=BCS_BoneSpace
    alphaBoolBlend {blendOption=Linear}
    alphaScaleBiasClamp {bMapRange=true, inRange {max=15}}
    boneToModify {boneName=dyn_hairTail_02}
    translation=<0.009 1>
   }
   n16 = "Transform (Modify) Bone - Bone: dyn_hairTail_05"(ComponentPose=@n15, Alpha=@n13, Translation="0.009,1,0", Rotation="-0.009,11.852,0.002", Scale="1,1,1") {
    rotationMode=BMM_Additive
    rotationSpace=BCS_BoneSpace
    translationMode=BMM_Additive
    translationSpace=BCS_BoneSpace
    alphaBoolBlend {blendOption=Linear}
    alphaScaleBiasClamp {bMapRange=true, inRange {max=15}}
    boneToModify {boneName=dyn_hairTail_05}
    rotation=<rot -0.009 11.8521 0.002>
    translation=<0.009 1>
   }
   n17 = "Transform (Modify) Bone - Bone: base_hairTail_00"!(ComponentPose=Local To Component(LocalPose=Use cached pose 'Start Pose'()), Alpha=Get Hair Up() -> (Hair Up:double), Rotation="0,-20,0") -> (Pose) {
    rotationMode=BMM_Additive
    rotationSpace=BCS_BoneSpace
    alphaBoolBlend {blendOption=Linear}
    boneToModify {boneName=base_hairTail_00}
    rotation {yaw=-20}
   }
   n18 = RigidBody!(ComponentPose=@n17, Alpha, OverrideWorldGravity:Vector=Get GravityOverride(), ExternalForce) -> (Pose) {
    bOverrideWorldGravity=true
    overridePhysicsAsset=&/F_MED_BalletAssassin_FaceAcc_PhysicsAsset
    simulationSpace=BaseBoneSpace
    alphaBoolBlend {blendOption=Linear}
    baseBoneRef {boneName=head}
    componentAppliedLinearAccClamp=<4000 2500 2500>
    componentLinearAccScale=<0.4 0.25 0.25>
    componentLinearVelScale=<0.4 0.25 0.25>
    simSpaceSettings {
     maxAngularAcceleration=1000
     maxAngularVelocity=3
     maxLinearAcceleration=700
     maxLinearVelocity=500
     velocityScaleZ=0
     worldAlpha=0.2
     externalLinearDragV=<0.2 0.2>
    }
   }
   Save cached pose 'Normal Dynamics'(Pose=Blend Poses by bool(BlendPose_0=Component To Local(ComponentPose=@n18), BlendPose_1=Component To Local(ComponentPose=@n16), bActiveValue=Get bRigidBodyEnabled()) {blendType=Linear}) {cacheName=Normal Dynamics}
   n19 = "Constraint - dyn_hairTail_00 (PP)"(ComponentPose=Local To Component(LocalPose=Use cached pose 'Start Pose'()), Alpha=1, ConstraintWeights_0=0.9, ConstraintWeights_1=0.1) {
    alphaBoolBlend {blendOption=Linear}
    boneToModify {boneName=dyn_hairTail_00}
    constraintSetup [{targetBone {boneName=head}}, {targetBone {boneName=spine_05}}]
    constraintWeights [1,1]
   }
   n20 = "Constraint - dyn_hairTail_01 (PP)"(ComponentPose=@n19, Alpha=1, ConstraintWeights_0=0.8, ConstraintWeights_1=0.2) {
    alphaBoolBlend {blendOption=Linear}
    boneToModify {boneName=dyn_hairTail_01}
    constraintSetup [{targetBone {boneName=head}}, {targetBone {boneName=spine_05}}]
    constraintWeights [1,1]
   }
   n21 = "Constraint - dyn_hairTail_02 (PP)"(ComponentPose=@n20, Alpha=1, ConstraintWeights_0=0.6, ConstraintWeights_1=0.4) {
    alphaBoolBlend {blendOption=Linear}
    boneToModify {boneName=dyn_hairTail_02}
    constraintSetup [{targetBone {boneName=head}}, {targetBone {boneName=spine_05}}]
    constraintWeights [1,1]
   }
   n22 = "Constraint - dyn_hairTail_03 (PP)"(ComponentPose=@n21, Alpha=1, ConstraintWeights_0=0.3, ConstraintWeights_1=0.7) {
    alphaBoolBlend {blendOption=Linear}
    boneToModify {boneName=dyn_hairTail_03}
    constraintSetup [{targetBone {boneName=head}}, {targetBone {boneName=spine_05}}]
    constraintWeights [1,1]
   }
   n23 = "Constraint - dyn_hairTail_04 (PPP)"(ComponentPose=@n22, Alpha=1, ConstraintWeights_0=0.3, ConstraintWeights_1=0.5, ConstraintWeights_2=0.2) {
    alphaBoolBlend {blendOption=Linear}
    boneToModify {boneName=dyn_hairTail_04}
    constraintSetup [{targetBone {boneName=head}}, {targetBone {boneName=spine_05}}, {targetBone {boneName=spine_04}}]
    constraintWeights [1,1,1]
   }
   n24 = "Constraint - dyn_hairTail_05 (PPP)"(ComponentPose=@n23, Alpha=1, ConstraintWeights_0=0.7, ConstraintWeights_1=0, ConstraintWeights_2=0.3) {
    alphaBoolBlend {blendOption=Linear}
    boneToModify {boneName=dyn_hairTail_05}
    constraintSetup [{targetBone {boneName=spine_05}}, {targetBone {boneName=spine_04}}, {targetBone {boneName=head}}]
    constraintWeights [1,1,1]
   }
   n25 = Get head_fwdBwd_rot()
   n26 = "Transform (Modify) Bone - Bone: dyn_hairTail_01"(ComponentPose=@n24, Alpha=@n25, Translation="0.015,-1,0.678", Rotation="0.002,-5,0.001", Scale="1,1,1") {
    rotationMode=BMM_Additive
    rotationSpace=BCS_BoneSpace
    translationMode=BMM_Additive
    translationSpace=BCS_BoneSpace
    alphaBoolBlend {blendOption=Linear}
    alphaScaleBiasClamp {bMapRange=true, inRange {max=0, min=-25}, outRange {max=0, min=1}}
    boneToModify {boneName=dyn_hairTail_01}
    rotation=<rot 0.002 -5 0.001>
    translation=<0.015 -1 0.678>
   }
   n27 = "Transform (Modify) Bone - Bone: dyn_hairTail_02"(ComponentPose=@n26, Alpha=@n25, Translation="-0.001,-0.467,-0.240", Rotation="-0.010,-2.865,0.003", Scale="1,1,1") {
    rotationMode=BMM_Additive
    rotationSpace=BCS_BoneSpace
    translationMode=BMM_Additive
    translationSpace=BCS_BoneSpace
    alphaBoolBlend {blendOption=Linear}
    alphaScaleBiasClamp {bMapRange=true, inRange {max=0, min=-25}, outRange {max=0, min=1}}
    boneToModify {boneName=dyn_hairTail_02}
    rotation=<rot -0.01 -2.865 0.003>
    translation=<-0.001 -0.467 -0.24>
   }
   n28 = Get head_fwdBwd_rot()
   n29 = "Transform (Modify) Bone - Bone: dyn_hairTail_01"(ComponentPose=@n27, Alpha=@n28, Translation="0,1,0", Rotation="0,0,0", Scale="1,1,1") {
    rotationMode=BMM_Additive
    rotationSpace=BCS_BoneSpace
    translationMode=BMM_Additive
    translationSpace=BCS_BoneSpace
    alphaBoolBlend {blendOption=Linear}
    alphaScaleBiasClamp {bMapRange=true, inRange {max=15}}
    boneToModify {boneName=dyn_hairTail_01}
    translation {y=1}
   }
   n30 = "Transform (Modify) Bone - Bone: dyn_hairTail_02"(ComponentPose=@n29, Alpha=@n28, Translation="0.009,1,0", Rotation="0,0,0", Scale="1,1,1") {
    rotationMode=BMM_Additive
    rotationSpace=BCS_BoneSpace
    translationMode=BMM_Additive
    translationSpace=BCS_BoneSpace
    alphaBoolBlend {blendOption=Linear}
    alphaScaleBiasClamp {bMapRange=true, inRange {max=15}}
    boneToModify {boneName=dyn_hairTail_02}
    translation=<0.009 1>
   }
   n31 = "Transform (Modify) Bone - Bone: dyn_hairTail_05"(ComponentPose=@n30, Alpha=@n28, Translation="0.009,1,0", Rotation="-0.009,11.852,0.002", Scale="1,1,1") {
    rotationMode=BMM_Additive
    rotationSpace=BCS_BoneSpace
    translationMode=BMM_Additive
    translationSpace=BCS_BoneSpace
    alphaBoolBlend {blendOption=Linear}
    alphaScaleBiasClamp {bMapRange=true, inRange {max=15}}
    boneToModify {boneName=dyn_hairTail_05}
    rotation=<rot -0.009 11.8521 0.002>
    translation=<0.009 1>
   }
   n32 = RigidBody(ComponentPose=@n31, OverrideWorldGravity=Get GravityOverride()) {
    bOverrideWorldGravity=true
    overridePhysicsAsset=&/F_MED_BalletAssassin_FaceAcc_Backpack_PhysicsAsset
    simulationSpace=BaseBoneSpace
    alphaBoolBlend {blendOption=Linear}
    baseBoneRef {boneName=head}
    componentAppliedLinearAccClamp=<4000 2500 2500>
    componentLinearAccScale=<0.4 0.25 0.25>
    componentLinearVelScale=<0.4 0.25 0.25>
    simSpaceSettings {
     maxAngularAcceleration=1000
     maxAngularVelocity=3
     maxLinearAcceleration=700
     maxLinearVelocity=500
     velocityScaleZ=0
     worldAlpha=0.2
     externalLinearDragV=<0.2 0.2>
    }
   }
   Save cached pose 'Backpack Dynamics'(Pose=Component To Local(ComponentPose=@n32)) {cacheName=Backpack Dynamics}
   n33 = RigidBody(ComponentPose=Local To Component(LocalPose=Use cached pose 'Start Pose'()), OverrideWorldGravity=Get GravityOverride()) {
    bOverrideWorldGravity=true
    overridePhysicsAsset=&/F_MED_BalletAssassin_FaceAcc_Cape_PhysicsAsset
    alphaBoolBlend {blendOption=Linear}
    componentAppliedLinearAccClamp=<3500 2000 4700>
    componentLinearAccScale=<0.35 0.2 0.47>
    componentLinearVelScale=<0.35 0.2 0.47>
    simSpaceSettings {
     maxAngularAcceleration=1000
     maxAngularVelocity=3
     maxLinearAcceleration=700
     maxLinearVelocity=500
     velocityScaleZ=0
     worldAlpha=0.2
     externalLinearDragV=<0.2 0.2>
    }
   }
   n34 = "Transform (Modify) Bone - Bone: dyn_hairTail_02"(ComponentPose=@n33, Alpha=1, Translation="0,-1,0", Rotation="0,0,0", Scale="1,1,1") {
    rotationMode=BMM_Additive
    rotationSpace=BCS_BoneSpace
    translationMode=BMM_Additive
    translationSpace=BCS_BoneSpace
    alphaBoolBlend {blendOption=Linear}
    boneToModify {boneName=dyn_hairTail_02}
    translation {y=-1}
   }
   n35 = "Transform (Modify) Bone - Bone: dyn_hairTail_03"(ComponentPose=@n34, Alpha=1, Translation="0,-3,0", Rotation="0,0,0", Scale="1,1,1") {
    rotationMode=BMM_Additive
    rotationSpace=BCS_BoneSpace
    translationMode=BMM_Additive
    translationSpace=BCS_BoneSpace
    alphaBoolBlend {blendOption=Linear}
    boneToModify {boneName=dyn_hairTail_03}
    translation {y=-3}
   }
   n36 = "Transform (Modify) Bone - Bone: dyn_hairTail_02"(ComponentPose=@n35, Alpha=Get Forward() -> (Forward:double), Translation="0,0,0", Rotation="0,-20,0", Scale="1,1,1") {
    rotationMode=BMM_Additive
    rotationSpace=BCS_BoneSpace
    translationMode=BMM_Additive
    translationSpace=BCS_BoneSpace
    alphaBoolBlend {blendOption=Linear}
    alphaScaleBiasClamp {bInterpResult=true}
    boneToModify {boneName=dyn_hairTail_02}
    rotation {yaw=-20}
   }
   n37 = RigidBody(ComponentPose=Local To Component(LocalPose=Use cached pose 'Start Pose'()), OverrideWorldGravity=Get GravityOverride()) {
    bOverrideWorldGravity=true
    overridePhysicsAsset=&F_MED_BalletAssassin_FaceAcc_Cape_PhysicsAsset
    alphaBoolBlend {blendOption=Linear}
    componentAppliedLinearAccClamp=<3500 2000 4700>
    componentLinearAccScale=<0.35 0.2 0.47>
    componentLinearVelScale=<0.35 0.2 0.47>
    simSpaceSettings {
     maxAngularAcceleration=1000
     maxAngularVelocity=3
     maxLinearAcceleration=700
     maxLinearVelocity=500
     velocityScaleZ=0
     worldAlpha=0.2
     externalLinearDragV=<0.2 0.2>
    }
   }
   n38 = "Transform (Modify) Bone - Bone: dyn_hairTail_02"(ComponentPose=@n37, Alpha=1, Translation="0,-1,0", Rotation="0,0,0", Scale="1,1,1") {
    rotationMode=BMM_Additive
    rotationSpace=BCS_BoneSpace
    translationMode=BMM_Additive
    translationSpace=BCS_BoneSpace
    alphaBoolBlend {blendOption=Linear}
    boneToModify {boneName=dyn_hairTail_02}
    translation {y=-1}
   }
   n39 = "Transform (Modify) Bone - Bone: dyn_hairTail_03"(ComponentPose=@n38, Alpha=1, Translation="0,-3,0", Rotation="0,0,0", Scale="1,1,1") {
    rotationMode=BMM_Additive
    rotationSpace=BCS_BoneSpace
    translationMode=BMM_Additive
    translationSpace=BCS_BoneSpace
    alphaBoolBlend {blendOption=Linear}
    boneToModify {boneName=dyn_hairTail_03}
    translation {y=-3}
   }
   Save cached pose 'Cape Dynamics'(Pose=Blend Poses by bool(BlendPose_0=Component To Local(ComponentPose=@n39), BlendPose_1=Component To Local(ComponentPose=@n36), bActiveValue=Get bIsCrouching()) {blendType=Linear}) {cacheName=Cape Dynamics}
   Use cached pose 'Normal Dynamics'()
  ]}
 EventGraph {nodes [
   Event BlueprintUpdateAnimation()
   > n1 = Sequence!(execute) -> (then_0, then_1, then_2:Exec, then_3:Exec, then_4:Exec, then_5:Exec, then_6:Exec)
   Set Hair Up(
    execute=@n1.then_1
    Hair Up:double=GetAlphaFromReferenceTransformDelta(SkeletalMeshComponent=Get Skel Mesh(), BoneName=spine_05, BaseName=head, InputRange=(X=0,Y=70), TwistAxis="0,0,1")
   ) -> (Output_Get:double)
   n2 = GetAlphaFromReferenceTransformDelta(
    SkeletalMeshComponent=Get Skel Mesh()
    BoneName=spine_05
    BaseName=head
    DeltaComponentType=EulerX
    InputRange=(X=-60,Y=60)
    OutputRange=(X=-1,Y=1)
    TwistAxis="0,0,1"
   )
   > Set Hair Side Rotation(Hair Side Rotation:Rotator=MakeRotator(Pitch=float * float(A=@n2, B=-120) {is=title K2Node_PromotableOperator})) -> (Output_Get:Rotator)
   n3 = Gravity Override State Function DEPRECATED(
    execute=@n1.then_0
    CharacterPartAnimInstance=Self-Reference()
    GameTimeInSeconds=GetGameTimeInSeconds()
    GravityOverrideParams=Get Gravity Override Params()
   )
   > Set GravityOverride(GravityOverride=@n3)
   n4 = Get is Wearing Backbling()
   n5 = Branch(execute=@n1.then_4, Condition="Nearly Equal (Float)"(A="To Float (Boolean)"(InBool=@n4), B=1, ErrorTolerance=0.1))
   n6 = GetDeltaTransformFromRefPose(self=GetOwnerSkeletalMeshComponent(), BoneName=head, BaseName=neck_01)
   > Set head_fwdBwd_rot(head_fwdBwd_rot:double=BreakRotator(InRot=BreakTransform(InTransform=@n6).Rotation).Yaw) -> (Output_Get:double)
   Set head_fwdBwd_rot(execute=@n5.else, head_fwdBwd_rot=0)
   n7 = CheckForBackbling()
   Set is Wearing Backbling(
    execute=@n1.then_2
    is Wearing Backbling:bool=AND Boolean!(A=@n7, B=NOT Boolean(A=@n7.bIsBackblingACape), C:bool=NOT Boolean(A=@n7.bIsBackblingLong)) -> (ReturnValue)
   ) -> (Output_Get:bool)
   > Set is Wearing Cape(is Wearing Cape:bool=OR Boolean(A=@n7.bIsBackblingACape, B=@n7.bIsBackblingLong)) -> (Output_Get:bool)
   n8 = Get bIsSkydiveDiveMode()
   n9 = Get bIsSkydiveFloating()
   Set is Skydiving(execute=@n1.then_3, is Skydiving:double="To Float (Boolean)"(InBool=OR Boolean(A=@n8, B=@n9))) -> (Output_Get:double)
   > Set is Parachuting Alpha(is Parachuting Alpha:double="To Float (Boolean)"(InBool=Get bIsParachuteOpened())) -> (Output_Get:double)
   n10 = PlayerMovement(execute=@n1.then_5, AnimInstance=Self-Reference(), XDotProductCoefficient=100, YDotProductCoefficient=100, ZDotProductCoefficient=100)
   n11 = BreakVector(InVec=MakeVector(X=@n10, Y=@n10.YMovement))
   > Set Left(Left:double="To Float (Boolean)"(InBool="float < float"(A=@n11, B=-50))) -> (Output_Get:double)
   > Set Right(Right:double="To Float (Boolean)"(InBool="float > float"(A=@n11, B=50))) -> (Output_Get:double)
   > Set Forward(Forward:double="To Float (Boolean)"(InBool="float > float"(A=@n11.Y, B=40))) -> (Output_Get:double)
   > Set Backward(Backward:double="To Float (Boolean)"(InBool="float < float"(A=@n11.Y, B=-40))) -> (Output_Get:double)
   n12 = "To Float (Boolean)"(InBool=Get bIsCrouching())
   Set Left Crouching(execute=@n1.then_6, Left Crouching:double=float * float(A=@n12, B=Get Left() -> (Left:double))) -> (Output_Get:double)
   > Set Forward Crouching(Forward Crouching:double=float * float(A=@n12, B=Get Forward())) -> (Output_Get:double)
   Event BlueprintInitializeAnimation()
   > Set Skel Mesh(Skel Mesh=GetOwnerSkeletalMeshComponent())
   TryGetPawnOwner()
  ]}
]
variables [
 Backward {type=number}
 Forward {type=number}
 Forward Crouching {type=number}
 Gravity Override Params {type=GravityOverrideParamsStruct}
 GravityOverride {type=Vector}
 Hair Side Rotation {type=Rotator}
 Hair Up {type=number}
 head_fwdBwd_rot {type=number}
 is Parachuting Alpha {type=number}
 is Skydiving {type=number}
 is Wearing Backbling {type=boolean}
 is Wearing Cape {type=boolean}
 IsShortBackblingOnAlpha {type=number}
 Left {type=number}
 Left Crouching {type=number}
 Right {type=number}
 Skel Mesh {type=SkeletalMeshComponent}
]
